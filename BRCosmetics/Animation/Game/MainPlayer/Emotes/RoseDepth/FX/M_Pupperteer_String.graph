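# Additive, unlit, two-sided material for Niagara mesh particles, drawing the puppeteer strings in the RoseDepth emote effect.
# SplineThicken sets string width, while a world-position sine wobble, pinned at both ends and randomized per particle, adds offset motion.
# Gradient masks eroded by StringGrow feed depth-faded opacity; scrolling noise plus a red end boost drives time-of-day tinted emissive.
# Dynamic parameters let particles control timing, growth, width, wobble, UV offset and fade distance.
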Material /BRCosmetics/Animation/Game/MainPlayer/Emotes/RoseDepth/FX/M_Pupperteer_String
graph {
 properties {
  bAutomaticallySetUsageInEditor=false
  bUsedWithNiagaraMeshParticles=true
  blendMode=BLEND_Additive
  shadingModel=MSM_Unlit
  translucencyLightingMode=TLM_VolumetricPerVertexDirectional
  twoSided=true
  usedShadingModels=Unlit
 }
 nodes [
  n1 = Multiply(
   A=ScalarParameter() {defaultValue=2.5, parameterName=Width}
   B=DynamicParameter() {defaultValue=<rgba 0 1 1 1>, paramNames [TimeOffset, StringGrow, StringWidth, WPO_Intensity]}.StringWidth
  )
  n2 = MaterialFunctionCall() {materialFunction=&LinearGradient}
  n3 = Multiply(A=Sine(None=Multiply(A=@n2.VGradient) {constB=0.5}), B=Constant() {r=1}) {constB=5}
  n4 = WorldPosition()
  n5 = DynamicParameter() {defaultValue=<rgba 0 1 1 1>, paramNames [TimeOffset, StringGrow, StringWidth, WPO_Intensity]}
  n6 = Multiply(A=Add(A=Time(), B=Multiply(A=@n5.TimeOffset) {constB=100}), B=ScalarParameter() {defaultValue=10, parameterName=WPO_Speed})
  n7 = VectorParameter() {parameterName=WPO_Tile, defaultValue=<rgba 15 25 100 0>}
  n8 = LinearInterpolate(Alpha=ParticleRandom()) {constA=0.5}
  n9 = Sine(None=Divide(A=Add(A=@n4.XYZ, B=@n6), B=Multiply(A=@n7.RGB, B=@n8)) {constB=4000})
  n10 = Multiply(A=ScalarParameter() {defaultValue=5, parameterName=WPO_Intensity}, B=@n5.WPO_Intensity)
  Add(
   A=MaterialFunctionCall(WidthBase=@n1, WidthTip=@n1) {materialFunction=&SplineThicken}.WorldPositionOffset
   B=Multiply(A=Saturate(None=@n3), B=Multiply(A=@n9, B=@n10))
  ) {constB=0}
  n11 = ParticleColor()
  n12 = MaterialFunctionCall() {materialFunction=&LinearGradient}
  n13 = Saturate(None=Sine(None=Multiply(A=@n12.UGradient) {constB=0.5}))
  n14 = Sine(None=Multiply(A=@n12.VGradient) {constB=0.5})
  n15 = LinearInterpolate(Alpha=@n12.VGradient)
  n16 = Multiply(A=Saturate(None=Multiply(A=@n14) {constB=20}), B=@n15)
  n17 = Saturate(None=Subtract(
    A=Multiply(A=@n13, B=@n16)
    B=DynamicParameter() {defaultValue=<rgba 0 1 1 1>, paramNames [TimeOffset, StringGrow, StringWidth, WPO_Intensity]}.StringGrow
   ))
  n18 = DynamicParameter() {parameterIndex=1, defaultValue=<rgba 0 25 1 1>, paramNames [U Offset, DepthDistance, Param3, Param4]}
  DepthFade(
   Opacity=Multiply(A=Multiply(A=@n11.A, B=@n17), B=ScalarParameter() {defaultValue=1, parameterName=OverallOpacity})
   FadeDistance=@n18.DepthDistance
  ) {fadeDistanceDefault=20}
  n19 = VectorParameter() {group=TOD, parameterName=Morning, defaultValue=<rgba 0.604 0.604 0.4228 1>}
  n20 = VectorParameter() {group=TOD, parameterName=Day, defaultValue=<rgba 1 1 1 1>}
  n21 = VectorParameter() {group=TOD, parameterName=Evening, defaultValue=<rgba 1 0.6511 0.371 1>}
  n22 = VectorParameter() {group=TOD, parameterName=Night, defaultValue=<rgba 0.5 0.7083 1 1>}
  n23 = Power(Base=OneMinus(None=MaterialFunctionCall() {materialFunction=&LinearGradient}.VGradient)) {constExponent=4}
  n24 = Constant3Vector() {constant=<rgba 1 0 0 0>}
  n25 = TextureCoordinate()
  n26 = Constant() {r=1}
  n27 = AppendVector(A=@n26, B=ScalarParameter() {defaultValue=2, parameterName=UV Scalar})
  n28 = MaterialFunctionCall(XY=Add(A=Multiply(A=@n25, B=@n27), B=AppendVector(A=@n26, B=@n18.U Offset))) {materialFunction=&Swizzle}
  n29 = TextureObject() {texture=/Game/Characters/Player/Male/Medium/Bodies/M_MED_Viper/Textures/Glitch/T_SoftNoise}
  n30 = Panner(Coordinate=@n28.YX) {speedX=0.1}
  n31 = Add(
   A=Multiply(A=Multiply(A=@n23, B=@n24)) {constB=10}
   B=Add(
    A=TextureSample(UVs=@n28.YX, Tex=@n29) {texture=&T_SoftNoise}.R
    B=TextureSample(UVs=@n30, Tex=@n29) {texture=&T_SoftNoise}.G
   )
  )
  n32 = Multiply(A=ScalarParameter() {defaultValue=1, parameterName=Emissive}, B=@n11.RGB)
  MaterialFunctionCall(Emissive=Multiply(
    A=MaterialFunctionCall(Color_Morning=@n19.RGB, Color_Day=@n20.RGB, Color_Evening=@n21.RGB, Color_Night=@n22.RGB) {materialFunction=&Color_Time_of_Day}.Result
    B=Multiply(A=@n31, B=@n32)
   )) {materialFunction=&MF_FX_EmissiveAdjustments}
  n33 = TextureSample(UVs=@n30) {samplerType=SAMPLERTYPE_LinearColor, texture=/Game/Effects/Fort_Effects/Gameplay/SmokeDustExplosions/Textures/T_RibbonSmokeMasked_03}
  Multiply(A=TextureSample(UVs=@n28.YX) {texture=/Game/Effects/Fort_Effects/Gameplay/SmokeDustExplosions/Textures/T_RibbonSmokeMasked_05}.R, B=@n33.R)
  Power(Base=@n15, Exp=ScalarParameter() {defaultValue=4}) {constExponent=4}
 ]
}
