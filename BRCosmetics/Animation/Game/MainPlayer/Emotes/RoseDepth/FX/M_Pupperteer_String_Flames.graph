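# Translucent, unlit, two-sided Niagara material, apparently for flames on puppeteer strings in the RoseDepth emote.
# Opacity combines a panning base texture distorted by two noises with optional dissolve and masks, UV shaping, view-angle fade, and depth fade on non-mobile deferred paths.
# Particle color, optionally tinted by a curve atlas gradient, drives emissive.
# Noise-driven world position offset adds optional spline thickening, with dynamic parameters supplying per-particle offsets and wave intensity.
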Material /BRCosmetics/Animation/Game/MainPlayer/Emotes/RoseDepth/FX/M_Pupperteer_String_Flames
graph {
 properties {
  bAutomaticallySetUsageInEditor=false
  bUsedWithNiagaraMeshParticles=true
  bUsedWithNiagaraRibbons=true
  bUsedWithNiagaraSprites=true
  blendMode=BLEND_Translucent
  shadingModel=MSM_Unlit
  twoSided=true
  usedShadingModels=Unlit
 }
 nodes [
  n1 = ScalarParameter() {defaultValue=20, parameterName=SplineBase}
  n2 = ScalarParameter() {defaultValue=15, parameterName=SplineTip}
  n3 = StaticBool()
  Add(
   A=LinearInterpolate(
    A=MaterialFunctionCall(WidthBase=@n1, WidthTip=@n2, Expand U or V UV Channel=@n3) {materialFunction=&SplineThicken}.WorldPositionOffset
    B=Constant()
    Alpha=ScalarParameter() {parameterName=Use_SplineThicken}
   )
   B=NamedRerouteUsage()
  )
  n4 = TextureCoordinate()
  n5 = Constant2Vector() {g=0.5, r=0.5}
  n6 = Constant() {r=0.5}
  n7 = Constant() {r=0.5}
  n8 = TextureCoordinate()
  n9 = CurveAtlasRowParameter(CurveTime=LinearInterpolate(
    A=SphereMask(A=@n4, B=@n5, Radius=@n6, Hardness=@n7)
    B=ComponentMask(None=@n8) {r=true}
    Alpha=ScalarParameter() {defaultValue=1, parameterName=Use_ColorCurveMask}
   )) {
   atlas=/BRCosmetics/Weapons/FORT_Melee/Pickaxe_LivelyDomino/FX/CAtlas_Pickaxe_LivelyDomino_ColorCurve
   curve=/BRCosmetics/Weapons/FORT_Melee/Pickaxe_LivelyDomino/FX/CV_Pickaxe_LivelyDomino_ColorCurve_Middle
   defaultValue=1
   group=Emissive
   parameterName=ColorCurve
   sortPriority=0
  }
  n10 = ParticleColor()
  n11 = Multiply(A=@n9, B=@n10.RGB)
  n12 = ScalarParameter() {parameterName=Use_ColorCurve}
  MaterialFunctionCall(Emissive=Multiply(A=LinearInterpolate(A=@n11, B=@n10.RGB, Alpha=@n12), B=ScalarParameter() {defaultValue=1, parameterName=All light})) {materialFunction=&MF_FX_EmissiveAdjustments}
  n13 = TextureSampleParameter2D(UVs=Panner(Coordinate=TextureCoordinate() {uTiling=0.6, vTiling=0.5}) {speedX=-0.25, speedY=0.15}) {
   group=DissolveMask
   parameterName=Dissolve B Texture
   samplerType=SAMPLERTYPE_Grayscale
   sortPriority=0
   texture=&LowResBlurredNoise
  }
  n14 = TextureSampleParameter2D(UVs=Panner(Coordinate=TextureCoordinate() {uTiling=0.6, vTiling=0.3}) {speedX=-0.25, speedY=0.15}) {
   group=DissolveMask
   parameterName=Dissolve A Texture
   samplerType=SAMPLERTYPE_LinearGrayscale
   sortPriority=0
   texture=/Game/Effects/Fort_Effects/Emotes/UltraLight/Texture/T_Plasma_C_Greyscale
  }
  n15 = ChannelMaskParameter(None=@n14.RGBA) {group=Base, maskChannel=Alpha, parameterName=DissolveTexture_RGBselect, sortPriority=5, defaultValue=<rgba 0 0 0 1>}
  n16 = MaterialFunctionCall(UVs=TextureCoordinate(), "Rotation Angle (0-1)"=ScalarParameter() {defaultValue=-0.4, group=DissolveMask, parameterName=DissolveMaskRotator}) {materialFunction=&CustomRotator}
  n17 = DynamicParameter() {parameterIndex=2, defaultValue=<rgba 1 0 1 0>, paramNames [BaseTextureSpeedX_Offset, BaseTextureX_Add, Param3, DissolveMaskOffset]}
  n18 = AppendVector(
   A=ScalarParameter() {group=DissolveMask, parameterName=DissloveMaskAdd_X}
   B=Add(A=ScalarParameter() {defaultValue=-0.3, group=DissolveMask, parameterName=DissolveMaskAdd_Y}, B=@n17.DissolveMaskOffset)
  )
  n19 = Add(
   A=LinearInterpolate(A=@n13.R, B=Saturate(None=@n15.R), Alpha=ScalarParameter() {defaultValue=1, parameterName=DissolveTextureSelect})
   B=ComponentMask(None=Add(A=@n16.Rotated Values, B=@n18)) {g=true}
  )
  n20 = ScalarParameter() {defaultValue=3, parameterName=Dissolve_Int}
  NamedRerouteDeclaration(None=Saturate(None=Multiply(A=@n19, B=@n20) {constB=3})) {name=DissloveMask, nodeColor=<rgba 0.0312 0 1 1>}
  n21 = DynamicParameter() {defaultValue=<rgba 1 0 0 6>, paramNames [Random Offset, BaseTextureV_Add, WPO Wave Intensity, SweepTexture_Add]}
  n22 = VectorParameter() {group=WPO, parameterName=WPO Side UV, sortPriority=6, defaultValue=<rgba 1 1 -0.1 1>}
  n23 = TextureCoordinate()
  n24 = Multiply(A=@n23, B=AppendVector(A=@n22.R, B=@n22.R))
  n25 = AppendVector(A=@n22.B, B=@n22.B)
  n26 = TextureSample(UVs=Add(A=AppendVector(A=@n21.Random Offset, B=@n21.Random Offset), B=Panner(Coordinate=@n24, Speed=@n25))) {samplerType=SAMPLERTYPE_Normal, texture=/Game/Characters/Player/Female/Medium/Bodies/F_MED_Bistro_Astronaut/FX/T_GaussianNoise_Normal}
  n27 = DynamicParameter() {defaultValue=<rgba 1 0 0 6>, paramNames [Random Offset, BaseTextureV_Add, WPO Wave Intensity, SweepTexture_Add]}
  n28 = VectorParameter() {group=WPO, parameterName=WPO Wave UV, sortPriority=6, defaultValue=<rgba 1 1 0 2>}
  n29 = AppendVector(A=@n28.R, B=@n28.R)
  n30 = Add(
   A=AppendVector(A=@n27.Random Offset, B=@n27.Random Offset)
   B=Panner(Coordinate=Multiply(A=@n23, B=@n29), Speed=AppendVector(A=@n28.B, B=@n28.B))
  )
  n31 = AppendVector(
   A=Multiply(A=ComponentMask(None=@n26.RGB) {g=true, r=true}, B=ScalarParameter() {defaultValue=1, group=WPO, parameterName=WPO Side Intensity}) {constB=10}
   B=Multiply(
    A=TextureSample(UVs=@n30) {texture=&T_Smooth_tile}.R
    B=ScalarParameter() {defaultValue=4, group=WPO, parameterName=WPO Wave Intensity}
   ) {constB=10}
  )
  n32 = ComponentMask(None=TextureCoordinate()) {r=true}
  n33 = LinearInterpolate(B=@n31, Alpha=@n32)
  n34 = ScalarParameter() {defaultValue=1, parameterName=Use_WPOMaskLerp}
  n35 = DynamicParameter() {defaultValue=<rgba 1 0 0 6>, paramNames [Random Offset, BaseTextureV_Add, WPO Wave Intensity, SweepTexture_Add]}
  NamedRerouteDeclaration(None=Multiply(A=LinearInterpolate(A=@n33, B=@n31, Alpha=@n34), B=@n35.WPO Wave Intensity) {constB=10}) {name=WPO, nodeColor=<rgba 0.0312 0 1 1>}
  OneMinus(None=@n32)
  ScalarParameter() {group=Base, parameterName=BaseTextureU_Add}
  ScalarParameter() {group=Base, parameterName=BaseTextureV_Add}
  n36 = DynamicParameter() {parameterIndex=2, defaultValue=<rgba 1 0 1 0>, paramNames [BaseTextureSpeedX_Offset, BaseTextureX_Add, Param3, DissolveMaskOffset]}
  n37 = DynamicParameter() {defaultValue=<rgba 1 0 0 6>, paramNames [Random Offset, BaseTextureV_Add, WPO Wave Intensity, SweepTexture_Add]}
  n38 = Multiply(
   A=Add(A=TextureCoordinate(), B=AppendVector(A=@n36.BaseTextureX_Add, B=@n37.BaseTextureV_Add))
   B=AppendVector(
    A=ScalarParameter() {defaultValue=1, group=Base, parameterName=BaseTextureU}
    B=ScalarParameter() {defaultValue=1, group=Base, parameterName=BaseTextureV}
   )
  )
  n39 = TextureCoordinate() {uTiling=0.8}
  n40 = Constant2Vector() {r=0.2}
  n41 = ScalarParameter() {defaultValue=-0.1, group=BaseNoise, parameterName=NoiseTexture01Panner_X}
  n42 = ScalarParameter() {group=BaseNoise, parameterName=NoiseTexture01Panner_Y}
  n43 = TextureSampleParameter2D(UVs=Panner(Coordinate=Add(A=@n39, B=@n40), Speed=AppendVector(A=@n41, B=@n42)) {speedX=-0.1}) {
   group=BaseNoise
   parameterName=NoiseTexture01
   samplerType=SAMPLERTYPE_Grayscale
   texture=/BRCosmetics/Weapons/FORT_Melee/Pickaxe_Cube_Ninja_Male/FX/T_CellNoise
  }
  n44 = LinearInterpolate(Alpha=ComponentMask(None=@n43.RGB) {g=true, r=true}) {constA=0.2}
  n45 = Add(A=TextureCoordinate(), B=Constant2Vector())
  n46 = AppendVector(
   A=ScalarParameter() {defaultValue=-0.5, group=BaseNoise, parameterName=NoiseTexture02Panner_X}
   B=ScalarParameter() {group=BaseNoise, parameterName=NoiseTexture02Panner_Y}
  )
  n47 = ComponentMask(None=TextureSampleParameter2D(UVs=Panner(Coordinate=@n45, Speed=@n46) {speedX=-0.5}) {group=BaseNoise, parameterName=NoiseTexture02, texture=/Game/Effects/Fort_Effects/Emotes/UltraLight/Texture/T_SoftNoise_BC7}.RGB) {g=true, r=true}
  n48 = ComponentMask(None=TextureCoordinate()) {r=true}
  n49 = Multiply(
   A=Multiply(A=Add(A=@n44, B=@n47), B=LinearInterpolate(Alpha=@n48) {constA=0.1})
   B=ScalarParameter() {defaultValue=0.15, group=BaseNoise, parameterName=Noise_Intensity}
  ) {constB=0.3}
  n50 = Multiply(A=ScalarParameter() {defaultValue=-0.6, group=Base, parameterName=BaseTexture_SpeedX}, B=@n36.BaseTextureSpeedX_Offset)
  n51 = ScalarParameter() {group=Base, parameterName=BaseTexture_SpeedY}
  n52 = TextureSampleParameter2D(UVs=Panner(Coordinate=Add(A=@n38, B=@n49), Speed=AppendVector(A=@n50, B=@n51)) {speedX=-0.6}) {group=Base, parameterName=BaseTexture, texture=/BRCosmetics/Weapons/FORT_Melee/Pickaxe_LivelyDomino/FX/T_Pickaxe_LivelyDomino_trail_Middle}
  n53 = ChannelMaskParameter(None=@n52.RGBA) {group=Base, maskChannel=Alpha, parameterName=BaseTexture_RGBselect, sortPriority=5, defaultValue=<rgba 0 0 0 1>}
  n54 = TextureSampleParameter2D(UVs=Panner(Coordinate=TextureCoordinate()) {speedX=-0.35, speedY=0.2}) {parameterName=BaseTextureLerp, samplerType=SAMPLERTYPE_Grayscale, texture=&LowResBlurredNoise}
  n55 = Multiply(A=Power(Base=@n53.R, Exp=ScalarParameter() {defaultValue=0.7, group=Base, parameterName=BaseTexture_power}), B=@n54.R)
  n56 = TextureSampleParameter2D() {parameterName=Mask, samplerSource=SSM_Clamp_WorldGroupSettings, texture=/Game/Effects/Fort_Effects/Textures/Mask/T_BaseLightMask_Half}
  n57 = LinearInterpolate(A=Multiply(A=@n55, B=@n56.R), B=@n55, Alpha=ScalarParameter() {defaultValue=1, group=DissolveMask, parameterName=Use_MaskTexture})
  n58 = NamedRerouteUsage()
  n59 = TextureCoordinate() {vTiling=0.5}
  n60 = ComponentMask(None=@n59) {r=true}
  n61 = OneMinus(None=@n60)
  n62 = ScalarParameter() {defaultValue=2, parameterName=Linear_Power}
  n63 = ScalarParameter() {defaultValue=1, parameterName=LinearLeft_Power}
  n64 = Multiply(A=Sine(None=ComponentMask(None=@n59) {g=true}), B=Multiply(A=Power(Base=@n61, Exp=@n62), B=Power(Base=@n60, Exp=@n63) {constExponent=1}))
  n65 = Multiply(
   A=LinearInterpolate(A=Multiply(A=@n57, B=@n58), B=@n57, Alpha=ScalarParameter() {group=DissolveMask, parameterName=Use_DissolveMask})
   B=Saturate(None=Multiply(A=@n64) {constB=2})
  )
  n66 = ComponentMask(None=TextureCoordinate()) {g=true}
  n67 = Multiply(
   A=@n65
   B=MaterialFunctionCall(In=OneMinus(None=@n66), Contrast=Constant() {r=0.3}) {materialFunction=&CheapContrast}.Result
  )
  n68 = ScalarParameter() {parameterName=Use_Mask}
  n69 = Multiply(
   A=Multiply(A=@n10.A, B=LinearInterpolate(A=@n67, B=@n65, Alpha=@n68))
   B=MaterialFunctionCall("FadeSensitivity (optional)"=ScalarParameter() {defaultValue=0.8, parameterName=MeshFade}) {materialFunction=&FadeBasedOnViewAngle}.Result
  )
  n70 = DepthFade(Opacity=@n69, FadeDistance=ScalarParameter() {defaultValue=5, parameterName=DepthFade})
  ShadingPathSwitch(Default=@n70, Deferred=@n70, Forward=@n69, Mobile=@n69)
 ]
}
