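# Niagara effect for the RoseDepth emote that draws marionette strings from a handle skeletal mesh to the player pawn, with glowing sprites on pawn bones.
# Strings bursts four meshes, each sampling a bone on both skeletal meshes every update, then orienting and stretching to span them.
# Curves and per-string random values drive string color and material parameters.
# Particle_Decals pins four pink-red, camera-offset sprites to pawn bones, fading in with system age.
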NiagaraSystem /BRCosmetics/Animation/Game/MainPlayer/Emotes/RoseDepth/FX/NS_Emote_RoseDepth
dependencies {
 usedDataInterfaces [&NiagaraDataInterfaceArrayFloat3, &NiagaraDataInterfaceCurve, &NiagaraDataInterfaceColorCurve]
 usedDynamicInputs []
 usedModules [
  &self:SystemSpawnScript.SystemScriptSource.SystemScriptGraph.NiagaraNodeAssignment_0.SetVariables_#1
  &SystemState
  &Timeline
 ]
}
emitters {
 Particle_Decals {
  emitterUpdateScript [
   &EmitterState(Life Cycle Mode=::NewEnumerator1, Loop Behavior=::NewEnumerator1, Loop Duration Mode=::NewEnumerator1, UseLoopDelay=false)
   &SpawnBurst_Instantaneous(Spawn Count=4, Use Spawn Probability=false, UseLoopCountLimit=false)
  ]
  particleSpawnScript [
   &PartitionParticles(Number Of Partitions=5)
   &InitializeParticle(
    Color=<rgba 1 0.0902 0.1002 1>
    Lifetime=3
    Sprite Rotation Mode=::NewEnumerator2
    Sprite Size Mode=::NewEnumerator3
    Uniform Sprite Size=27.5
    UsePositionOffset=false
    Write Material Random=true
   )
   &CameraOffset(Camera Offset Amount=15)
   &DynamicMaterialParameters(
    Index 0 Param 1=1
    Index 0 Param 2=1
    Module.Index0Param0Write=true
    Module.Index0Param1Write=true
    Module.Index0Param2Write=true
    Module.Index0Param3Write=false
   )
   /SetVariables(Particles.EnergyDynam:Vector4f=<2 1 8 1>) {isSetParameters=true}
  ]
  particleUpdateScript [
   &ParticleState(Kill Particles When Lifetime Has Elapsed=false, Let Infinitely Lived Particles Die When Emitter Deactivates=true)
   &ScaleColor(
    Scale Alpha=&FloatFromCurve(
     FloatCurve=DataInterface() {propertyValues {Curve {keys [{tangentMode=RCTM_User, time=0, value=0}, {tangentMode=RCTM_User, time=1.8, value=1}]}}}
     CurveIndex=$Engine.System.Age
     Scale Curve=2
    )
    Scale RGB=&MakeVectorFromLinearColorRGB(Color=<rgba 1 1 1 1>)
    ScaleA=true
    ScaleRGB=true
   )
   /SetVariables(Particles.EnergyColor:LinearColor=&MultiplyLinearColor(Linear Color=$Particles.Color, Scale Factor=&MakeVector4_FromFloat(Float=30))) {isSetParameters=true}
   &SkeletalMeshLocation(
    Bone / Socket Index=$Particles.PartitionParticles.Partition
    Bone Sampling Mode=::NewEnumerator4
    MaskSpawnGroup=false
    Skeletal Mesh=$User.PlayerPawn
   )
   /SetVariables(Particles.Glow Colow:LinearColor=&MultiplyLinearColor(
     Linear Color=<rgba 1 0.0902 0.1002 1>
     Scale Factor=&MakeFloat4FromVectorAndFloat(
      "Vector (RGB)"=<1 1 1>
      "Float (Alpha)"=&FloatFromCurve(
       FloatCurve=DataInterface() {propertyValues {Curve {keys [{tangentMode=RCTM_User, time=0, value=0}, {tangentMode=RCTM_User, time=1.8, value=1}]}}}
       CurveIndex=$Engine.System.Age
       Scale Curve=0.25
      )
     )
    )) {isSetParameters=true}
  ]
  rendererClasses [&NiagaraSpriteRendererProperties]
  renderers [
   {rendererClass=&NiagaraSpriteRendererProperties}
   {rendererClass=&NiagaraSpriteRendererProperties, rendererIndex=1}
   {rendererClass=&NiagaraSpriteRendererProperties, rendererIndex=2}
  ]
 }
 Strings {
  emitterUpdateScript [
   &EmitterState(Life Cycle Mode=::NewEnumerator1, Loop Behavior=::NewEnumerator1, Loop Duration Mode=::NewEnumerator1, UseLoopDelay=false)
   &SpawnBurst_Instantaneous(Spawn Count=4, Use Spawn Probability=false, UseLoopCountLimit=false)
  ]
  particleSpawnScript [
   &PartitionParticles(Number Of Partitions=4)
   &InitializeParticle(Mesh Scale Mode=::NewEnumerator3, UsePositionOffset=false, Write Material Random=true)
   &CameraOffset(Camera Offset Amount=5)
  ]
  particleUpdateScript [
   &ParticleState(Kill Particles When Lifetime Has Elapsed=false, Let Infinitely Lived Particles Die When Emitter Deactivates=true)
   &SkeletalMeshLocation(
    Bone / Socket Index=$Particles.PartitionParticles.Partition
    Bone Sampling Mode=::NewEnumerator4
    MaskSpawnGroup=false
    Orientation Sampling=::NewEnumerator1
    Sampled Position Offset=&SelectVectorFromArray(
     Direct Array Index=$Particles.PartitionParticles.Partition
     Vector Selection Array=DataInterface() {propertyValues {FloatData [<0 0 0>, <0 0 0>, <0 0 0>, <0 0 3>, <0 0 3>]}}
    )
    Skeletal Mesh=$System.Marionette Handle
   )
   &SkeletalMeshLocation(
    Bone / Socket Index=$Particles.PartitionParticles.Partition
    Bone Sampling Mode=::NewEnumerator4
    MaskSpawnGroup=false
    Orientation Sampling=::NewEnumerator1
    Position Sampling=::NewEnumerator2
    Sampled Position Offset=&SelectVectorFromArray(
     Direct Array Index=$Particles.PartitionParticles.Partition
     Vector Selection Array=DataInterface() {propertyValues {FloatData [<-20 0 -5>, <0 0 0>, <20 0 5>, <0 0 0>, <0 0 0>]}}
    )
    Skeletal Mesh=$User.PlayerPawn
   ) {module=SkeletalMeshLocation001}
   /SetVariables(Particles.MeshOrientation=&OrientMeshToVector(
     Look At Direction=&Subtract_Position(A=$Particles.SkeletalMeshLocation001.SampledPosition, B=$Particles.SkeletalMeshLocation.SampledPosition)
     Reference Axis Vector=<1 0 0>
    )) {isSetParameters=true}
   &ScaleColor(
    Scale RGB=&Multiply_VectorByFloat(
     Vector=<1 1 1>
     Float=&FloatFromCurve(FloatCurve=DataInterface() {propertyValues {Curve {keys [
          {arriveTangent=1431.2385, interpMode=RCIM_Cubic, leaveTangent=1431.2385, tangentMode=RCTM_User, time=0, value=0}
          {arriveTangent=1.1179, interpMode=RCIM_Cubic, leaveTangent=1.1179, tangentMode=RCTM_User, time=0.1, value=50}
          {arriveTangent=0.0224, interpMode=RCIM_Cubic, leaveTangent=0.0224, tangentMode=RCTM_User, time=1, value=0.1}
         ]}}})
    )
    ScaleA=true
    ScaleRGB=true
   ) {enabled=false}
   &ScaleColor(
    Linear Color Curve=DataInterface() {propertyValues {
      AlphaCurve {keys [<at 0 2>]}
      BlueCurve {keys [<at 0.5156 0.5887>,<at 0.6213 0>,<at 0.6386 0.0589>,<at 1 0.15>]}
      GreenCurve {keys [<at 0.5156 0.5555>,<at 0.6213 0>,<at 0.6386 0.0556>,<at 1 0.15>]}
      RedCurve {keys [<at 0.5156 10>,<at 0.6213 0.5865>,<at 0.6386 1>,<at 1 0.15>]}
     }}
    Scale Mode=::NewEnumerator2
   ) {module=ScaleColor001}
   &ScaleMeshSize(Scale Factor=&Multiply_VectorByFloat(
     Vector=&MakeVector(
      X=&Divide_Float(
       A=&VectorLength(Vector=&Subtract_Position(A=$Particles.SkeletalMeshLocation001.SampledPosition, B=$Particles.SkeletalMeshLocation.SampledPosition))
       B=100
      )
      Y=1
      Z=1
     )
     Float=&FloatFromCurve(FloatCurve=DataInterface() {propertyValues {Curve {keys [<at 0 0 RCIM_Cubic>,<at 0.1 1 RCIM_Cubic>]}}})
    ))
   &DynamicMaterialParameters(
    Index 0 Param 1=&RandomRangeFloat(Minimum=0, Maximum=1)
    Index 0 Param 2=&FloatFromCurve(
     FloatCurve=DataInterface() {propertyValues {Curve {keys [
         {arriveTangent=-0.5798, time=0, value=1}
         {arriveTangent=-0.0046, interpMode=RCIM_Cubic, leaveTangent=-0.0046, tangentMode=RCTM_User, time=0.25, value=1}
         {arriveTangent=-5.1646, interpMode=RCIM_Cubic, leaveTangent=-5.1646, tangentMode=RCTM_User, time=0.75, value=0.2}
        ]}}}
     CurveIndex=$Particles.Age
    )
    Index 0 Param 3=&Lerp_Float(
     A=2
     B=1
     Alpha=&FloatFromCurve(
      FloatCurve=DataInterface() {propertyValues {Curve {keys [<at 0.0673 1 RCIM_Cubic>, {arriveTangent=-1.0722, interpMode=RCIM_Cubic, leaveTangent=-1.0722, time=0.108, value=0.35}, <at 1 0 RCIM_Cubic>]}}}
      CurveIndex=$Particles.Age
     )
    )
    Index 0 Param 4=&FloatFromCurve(
     FloatCurve=DataInterface() {propertyValues {Curve {keys [<at 0 10 RCIM_Cubic>, {arriveTangent=-55.4308, tangentMode=RCTM_User, time=0.75, value=1}, <at 1 1 RCIM_Cubic>]}}}
     CurveIndex=$Particles.Age
     Scale Curve=0.5
    )
    Index 1 Param 1=&RandomRangeFloat(Minimum=0, Maximum=1)
    Index 1 Param 2=25
    Module.Index0Param0Write=true
    Module.Index0Param1Write=true
    Module.Index0Param2Write=true
    Module.Index0Param3Write=true
    Module.Index1Param0Write=true
    Module.Index1Param1Write=true
    Module.Index1Param2Write=false
    Module.Index1Param3Write=false
    Write Parameter Index 1=true
   )
   /SetVariables(
    Particles.SmokeRandomA:NiagaraFloat=&RandomRangeFloat(Minimum=0, Maximum=1)
    Particles.SmokeRandomB:NiagaraFloat=&RandomRangeFloat(Minimum=0, Maximum=1)
    Particles.SmokeRandomC:NiagaraFloat=&RandomRangeFloat(Minimum=0, Maximum=1)
    Particles.SmokeRandomD:NiagaraFloat=&RandomRangeFloat(Minimum=0, Maximum=1)
   ) {isSetParameters=true}
   /SetVariables(Particles.Smoke Dynamic Parameter:Vector4f=&MakeVector4(
     X=&self:Random_per_Ribbon(
      Ribbon_0:NiagaraFloat=$Particles.SmokeRandomA
      Ribbon ID:NiagaraInt32=$Particles.PartitionParticles.Partition
      Ribbon_1:NiagaraFloat=$Particles.SmokeRandomB
      Ribbon_2:NiagaraFloat=$Particles.SmokeRandomC
      Ribbon_3:NiagaraFloat=$Particles.SmokeRandomD
     )
     Y=&self:Random_per_Ribbon(
      Ribbon_0:NiagaraFloat=-0.2
      Ribbon ID:NiagaraInt32=$Particles.PartitionParticles.Partition
      Ribbon_1:NiagaraFloat=-0.1
      Ribbon_2:NiagaraFloat=0
      Ribbon_3:NiagaraFloat=0.2
     )
     Z=0
     W=0
    )) {isSetParameters=true}
   /SetVariables(Particles.Smoke Color:LinearColor=&MakeLinearColorFromVectorAndFloat("Vector (RGB)"=<1 1 1>, "Float (Alpha)"=&RandomRangeFloat(Minimum=40, Maximum=60))) {enabled=false, isSetParameters=true}
   /SetVariables(Particles.SmokeScale:Vector3f=&Multiply_Vector(
     A=<1.5 1 1>
     B=&MakeVector(
      X=&FloatFromCurve(FloatCurve=DataInterface() {propertyValues {Curve {keys [<at 0.2 0>,<at 0.5 1>]}}})
      Y=1
      Z=1
     )
    )) {enabled=false, isSetParameters=true}
  ]
  rendererClasses [&NiagaraMeshRendererProperties]
  renderers [
   {rendererClass=&NiagaraMeshRendererProperties, rendererIndex=0}
   {rendererClass=&NiagaraMeshRendererProperties, rendererIndex=1}
  ]
 }
}
summary {
 systemName=NS_Emote_RoseDepth
 emitters [
  {emitterName=Strings, rendererClasses [&NiagaraMeshRendererProperties]}
  {emitterName=Particle_Decals, rendererClasses [&NiagaraSpriteRendererProperties]}
 ]
 userVariables [
  User.IsStringConnecting {type=NiagaraBool, defaultValue=true}
  User.MaskDBNO {type=NiagaraBool, defaultValue=true}
  User.PlayerPawn {
   type=NiagaraDataInterfaceSkeletalMesh
   defaultValue {
    struct=NiagaraExt_VariableValue_DataInterface
    value {dataInterface=&self:NiagaraDataInterfaceSkeletalMesh_0}
   }
  }
  User.TightDistance {type=NiagaraFloat, defaultValue=1, description=""}
 ]
}
